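# Looping interior ambience cue for the Chrome Jungle Temple chamber point of interest.
# A quality-level node selects the looping quad wave for the first two levels and the stereo version for the last two.
# It uses Athena ambient attenuation and the indoor ambient sound class at reduced volume.
# A concurrency override allows one instance per owner and prevents new ones from starting while it plays.
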SoundCue /Game/Sounds/Artemis/POI/Chrome_JungleTemple/Ambience_Chrome_JungleTemple_Chamber_Cue
root=QualityLevel_0
nodes [
 {id=QualityLevel_0, children [WavePlayer_2,WavePlayer_2,WavePlayer_3,WavePlayer_3]}
 {
  id=WavePlayer_2
  properties {bLooping=true, soundWaveAssetPtr=&/Ambience_Chrome_JungleTemple_Chamber_Int_Loop_Quad}
 }
 {
  id=WavePlayer_3
  properties {bLooping=true, soundWaveAssetPtr=&/Ambience_Chrome_JungleTemple_Chamber_Int_Loop_Stereo}
 }
]
properties {
 attenuationSettings=&Fort_Ambient_Athena_Att
 bOverrideConcurrency=true
 maxDistance=2097152
 soundClassObject=&Fort_Ambient_Indoors_Athena
 volumeMultiplier=0.667
 concurrencyOverrides {bLimitToOwner=true, maxCount=1, resolutionRule=PreventNew, retriggerTime=2}
}
