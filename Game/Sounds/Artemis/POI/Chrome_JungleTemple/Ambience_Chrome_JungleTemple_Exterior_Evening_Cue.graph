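# Looping exterior evening ambience cue for the Chrome Jungle Temple POI, using outdoor ambient sound class and attenuation.
# A looping quad "Mangled" wave feeds two nested, endlessly looping Envelopers: one varies pitch over a 60-second cycle, the other gently dips volume over 20 seconds.
# The unequal periods layer slow drift onto the single loop.
# A quality node routes this chain to its first two levels, leaving the remaining two empty.
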SoundCue /Game/Sounds/Artemis/POI/Chrome_JungleTemple/Ambience_Chrome_JungleTemple_Exterior_Evening_Cue
root=QualityLevel_0
nodes [
 {
  id=Enveloper_1
  children [WavePlayer_0]
  properties {
   bLoop=true
   bLoopIndefinitely=true
   loopEnd=60
   pitchCurve {editorCurveData {keys [
      {arriveTangent=-0.0013, interpMode=RCIM_Cubic, time=0}
      {arriveTangent=-0.0006, interpMode=RCIM_Cubic, leaveTangent=-0.0006, time=15.6395, value=0.8812}
      <at 30.6395 0.9812 RCIM_Cubic>
      {arriveTangent=0.0006, interpMode=RCIM_Cubic, leaveTangent=0.0006, time=45.6395, value=0.8812}
      {time=60}
     ]}}
   volumeCurve {editorCurveData {keys [<at 0 1>]}}
  }
 }
 {
  id=Enveloper_3
  children [Enveloper_1]
  properties {
   bLoop=true
   bLoopIndefinitely=true
   loopEnd=20
   pitchCurve {editorCurveData {keys [{time=0}]}}
   volumeCurve {editorCurveData {keys [
      {arriveTangent=-0.01, interpMode=RCIM_Cubic, leaveTangent=-0.01, time=0, value=1}
      <at 10 0.9 RCIM_Cubic>
      {arriveTangent=0.01, interpMode=RCIM_Cubic, leaveTangent=0.01, time=20, value=1}
     ]}}
  }
 }
 {id=QualityLevel_0, children [Enveloper_3,Enveloper_3,null,null]}
 {
  id=WavePlayer_0
  properties {bLooping=true, soundWaveAssetPtr=&/Ambience_Chrome_JungleTemple_Evening_Mangled_Loop_Quad}
 }
]
properties {
 attenuationSettings=&Fort_Ambient_Athena_Att
 maxDistance=2097152
 pitchMultiplier=0.95
 soundClassObject=&Fort_Ambient_Outdoors_Athena
 volumeMultiplier=0.301
 preEffectBusSends [{}]
}
