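# Sound cue for the looping nighttime exterior ambience at the Chrome Jungle Temple POI, using outdoor ambient class and attenuation settings.
# A quality switch plays a quad "Mangled" loop for the first two levels and a stereo version for the last two.
# Each loop runs through nested, endlessly looping envelopers: a 60-second pitch modulation inside a gentle 30-second volume dip, adding slow movement to the static loop.
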SoundCue /Game/Sounds/Artemis/POI/Chrome_JungleTemple/Ambience_Chrome_JungleTemple_Exterior_Night_Cue
root=QualityLevel_1
nodes [
 {
  id=Enveloper_3
  children [WavePlayer_2]
  properties {
   bLoop=true
   bLoopIndefinitely=true
   loopEnd=60
   pitchCurve {editorCurveData {keys [
      {arriveTangent=-0.0013, interpMode=RCIM_Cubic, time=0}
      {arriveTangent=-0.0006, interpMode=RCIM_Cubic, leaveTangent=-0.0006, time=15.5067, value=0.9}
      {arriveTangent=0.0017, interpMode=RCIM_Cubic, leaveTangent=0.0017, time=30.6395, value=0.9812}
      {arriveTangent=0.0006, interpMode=RCIM_Cubic, leaveTangent=0.0006, time=45.5067, value=0.95}
      {time=60}
     ]}}
   volumeCurve {editorCurveData {keys [<at 0 1>]}}
  }
 }
 {
  id=Enveloper_4
  children [WavePlayer_3]
  properties {
   bLoop=true
   bLoopIndefinitely=true
   loopEnd=60
   pitchCurve {editorCurveData {keys [
      {arriveTangent=-0.0013, interpMode=RCIM_Cubic, time=0}
      {arriveTangent=-0.0007, interpMode=RCIM_Cubic, leaveTangent=-0.0007, time=16.7492, value=0.8782}
      <at 31.7492 0.9782 RCIM_Cubic>
      {arriveTangent=0.0008, interpMode=RCIM_Cubic, leaveTangent=0.0008, time=46.7492, value=0.8782}
      {time=60}
     ]}}
   volumeCurve {editorCurveData {keys [<at 0 1>]}}
  }
 }
 {
  id=Enveloper_5
  children [Enveloper_4]
  properties {
   bLoop=true
   bLoopIndefinitely=true
   loopEnd=30
   pitchCurve {editorCurveData {keys [{time=0}]}}
   volumeCurve {editorCurveData {keys [
      <at 0 1 RCIM_Cubic>
      {arriveTangent=0, interpMode=RCIM_Cubic, leaveTangent=0, tangentMode=RCTM_User, time=15, value=0.9}
      <at 30 1 RCIM_Cubic>
     ]}}
  }
 }
 {
  id=Enveloper_6
  children [Enveloper_3]
  properties {
   bLoop=true
   bLoopIndefinitely=true
   loopEnd=30
   pitchCurve {editorCurveData {keys [{time=0}]}}
   volumeCurve {editorCurveData {keys [
      <at 0 1 RCIM_Cubic>
      {arriveTangent=0, interpMode=RCIM_Cubic, leaveTangent=0, tangentMode=RCTM_User, time=15, value=0.9}
      <at 30 1 RCIM_Cubic>
     ]}}
  }
 }
 {id=QualityLevel_1, children [Enveloper_6,Enveloper_6,Enveloper_5,Enveloper_5]}
 {
  id=WavePlayer_2
  properties {bLooping=true, soundWaveAssetPtr=&/Ambience_Chrome_JungleTemple_Night_Mangled_Loop_Quad}
 }
 {
  id=WavePlayer_3
  properties {bLooping=true, soundWaveAssetPtr=&/Ambience_Chrome_JungleTemple_Night_Mangled_Loop_Stereo}
 }
]
properties {
 attenuationSettings=&Fort_Ambient_Athena_Att
 maxDistance=2097152
 pitchMultiplier=0.9
 soundClassObject=&Fort_Ambient_Outdoors_Athena
 volumeMultiplier=0.401
}
